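# Physics asset for the CordSyrup backpack cosmetic, giving physics shapes to the pack and its three-bone dangling tail.
# One larger capsule covers the pack bone, and a small capsule sits on each tail bone.
# Constraints chain the tail from the pack, with a wide swing limit at the root, tighter limits between segments, and zero-degree twist limits throughout.
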PhysicsAsset /BRCosmetics/Accessories/FORT_Backpacks/Backpack_F_MED_CordSyrup/Meshes/F_MED_CordSyrup_Pack_Physics
bodies {
 dyn_Tail_01 [{shapeType=Capsule, length=2, radius=1, shapeName=dyn_Tail_01_capsule, center=<2 0 0>, rotation=<rot 90 0 0>}]
 dyn_Tail_02 [{shapeType=Capsule, length=2, radius=1, shapeName=dyn_Tail_02_capsule, center=<2 0 0>, rotation=<rot 90 0 0>}]
 dyn_Tail_03 [{shapeType=Capsule, length=2, radius=1, shapeName=dyn_Tail_03_capsule, center=<2 0 0>, rotation=<rot 90 0 0>}]
 pack [{shapeType=Capsule, length=8.0216, radius=8.1922, center=<0.0033 -4.3849 -8.8926>, rotation=<rot 0 -90 -90>}]
}
constraints [
 {bone1Name=dyn_Tail_01, bone2Name=pack, swing1LimitDegrees=85, swing2LimitDegrees=85, twistLimitDegrees=0}
 {bone1Name=dyn_Tail_02, bone2Name=dyn_Tail_01, swing1LimitDegrees=30, swing2LimitDegrees=30, twistLimitDegrees=0}
 {bone1Name=dyn_Tail_03, bone2Name=dyn_Tail_02, swing1LimitDegrees=30, swing2LimitDegrees=30, twistLimitDegrees=0}
]
